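# Physics asset for the OxideHoard medium male player body, covering a dynamic tie, a holster, the torso and both thighs.
# Five tie bodies form a chain from spine_05 with narrow swing and twist limits, loosest in swing at the root.
# Capsules on pelvis, spine and thighs give simplified body volumes, with no constraints between spine segments.
# Holster and thighs link to the pelvis with free swing and twist.
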PhysicsAsset /BRCosmetics/Characters/Player/Male/Medium/Bodies/M_MED_OxideHoard/Meshes/M_MED_OxideHoard_Physics
bodies {
 dyn_tie_01 [{shapeType=Capsule, length=5.9832, radius=3.9627, shapeName=dyn_tie_03_capsule, center=<9.3522 0 3.0298>, rotation=<rot 90 -90 0>}]
 dyn_tie_02 [{shapeType=Capsule, length=5.5725, radius=4.3734, shapeName=dyn_tie_03_capsule, center=<10.199 0 2.7764>, rotation=<rot 90 -90 0>}]
 dyn_tie_03 [
  {shapeType=Capsule, length=5.9341, radius=4.0118, shapeName=dyn_tie_03_capsule, center=<6.7351 0 2.4192>, rotation=<rot 90 -90 0>}
  {shapeType=Capsule, length=11.0637, radius=4.7101, shapeName=dyn_tie_03_capsule, center=<6.2328 0 2.9062>, rotation=<rot 86.4053 0 0>}
 ]
 dyn_tie_04 [{shapeType=Capsule, length=6.9332, radius=3.972, shapeName=dyn_tie_03_capsule, center=<4.1306 0 3.0202>, rotation=<rot 89.2669 -180 -180>}]
 dyn_tie_05 [{
   shapeType=Box
   extentX=7.6243
   extentY=3.6594
   extentZ=7.3573
   shapeName=dyn_tie_05_box
   center=<3.0133 -0.0282 -0.0262>
   rotation=<rot 0 -89.4397 -87.655>
  }]
 holster [{
   shapeType=Capsule
   length=13.3631
   radius=5.0171
   shapeName=holster_capsule
   center=<3.2486 -0.3897 2.0458>
   rotation=<rot 0 -91.7245 -81.2354>
  }]
 pelvis [{shapeType=Capsule, length=6.6167, radius=14.6832, shapeName=pelvis_capsule, center=<-0.6994 0 0>}]
 spine_01 [{shapeType=Capsule, length=7.8496, radius=15.6775, shapeName=pelvis_capsule, center=<0 -1.4483 0>, rotation=<rot 90 -0.0001 0>}]
 spine_02 [{shapeType=Capsule, length=4.9226, radius=14.4166, shapeName=pelvis_capsule, center=<1.5028 0 0>}]
 spine_03 [{shapeType=Capsule, length=5.0462, radius=14.293, shapeName=pelvis_capsule}]
 spine_04 [{shapeType=Capsule, length=9.6941, radius=13.3231, shapeName=pelvis_capsule}]
 spine_05 [{shapeType=Capsule, length=13.8398, radius=11.6611, shapeName=pelvis_capsule, center=<-3.6459 0 0>}]
 thigh_l [{
   shapeType=Capsule
   length=26.0592
   radius=9.5766
   shapeName=thigh_l_capsule
   center=<-12.9205 2.533 0>
   rotation=<rot 88.3953 2.0356 -179.9987>
  }]
 thigh_r [{shapeType=Capsule, length=26.0592, radius=9.5766, shapeName=thigh_r_capsule, center=<12.9205 -2.533 0>, rotation=<rot 87.6289 2.0225 0.0004>}]
}
constraints [
 {bone1Name=dyn_tie_01, bone2Name=spine_05, swing1LimitDegrees=25, swing2LimitDegrees=25, twistLimitDegrees=15}
 {bone1Name=dyn_tie_02, bone2Name=dyn_tie_01, swing1LimitDegrees=15, swing2LimitDegrees=15, twistLimitDegrees=15}
 {bone1Name=dyn_tie_03, bone2Name=dyn_tie_02, swing1LimitDegrees=15, swing2LimitDegrees=15, twistLimitDegrees=15}
 {bone1Name=dyn_tie_04, bone2Name=dyn_tie_03, swing1LimitDegrees=15, swing2LimitDegrees=15, twistLimitDegrees=15}
 {bone1Name=dyn_tie_05, bone2Name=dyn_tie_04, swing1LimitDegrees=15, swing2LimitDegrees=15, twistLimitDegrees=15}
 {
  bone1Name=holster
  bone2Name=pelvis
  swing1LimitDegrees=45
  swing1Motion=Free
  swing2LimitDegrees=45
  swing2Motion=Free
  twistLimitDegrees=45
  twistMotion=Free
 }
 {
  bone1Name=thigh_r
  bone2Name=pelvis
  swing1LimitDegrees=45
  swing1Motion=Free
  swing2LimitDegrees=45
  swing2Motion=Free
  twistLimitDegrees=45
  twistMotion=Free
 }
 {
  bone1Name=thigh_l
  bone2Name=pelvis
  swing1LimitDegrees=45
  swing1Motion=Free
  swing2LimitDegrees=45
  swing2Motion=Free
  twistLimitDegrees=45
  twistMotion=Free
 }
]
