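# Post-process style animation Blueprint for the OxideHoard medium male cosmetic body, copying the attached parent's pose and adding secondary motion.
# The pose runs through a Control Rig, RigidBody simulation, spine_01–05 scale-replace nodes, then spring-driven AnimDynamics on four sleeve bones.
# Each update computes spine scales with Y/Z clamped, plus sleeve alphas inversely remapped from upper-arm roll.
# The graph shows no explicit links from these variables to node inputs.
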AnimBlueprint /BRCosmetics/Characters/Player/Male/Medium/Bodies/M_MED_OxideHoard/Meshes/M_MED_OxideHoard_AnimBP
target_skeleton=&/M_MED_OxideHoard_Skeleton
graphs [
 AnimGraph {nodes [
   n1 = Control Rig(Source=Copy Pose From Mesh() {bUseAttachedParent=true}) {controlRigAssetReference {blueprintRigClass=&/M_MED_OxideHoard_CtrlRig.M_MED_OxideHoard_CtrlRig_C}}
   n2 = RigidBody(ComponentPose=Local To Component(LocalPose=@n1)) {
    bFreezeIncomingPoseOnStart=true
    overridePhysicsAsset=&/M_MED_OxideHoard_Physics
    simulationSpace=BaseBoneSpace
    baseBoneRef {boneName=spine_05}
    componentAppliedLinearAccClamp=<1000 1000 1000>
    simSpaceSettings {maxAngularAcceleration=50, maxAngularVelocity=10, maxLinearAcceleration=50, maxLinearVelocity=10, worldAlpha=1}
   }
   n3 = "Transform (Modify) Bone - Bone: spine_03"(
    ComponentPose="Transform (Modify) Bone - Bone: spine_02"(
     ComponentPose="Transform (Modify) Bone - Bone: spine_01"(ComponentPose=@n2, Alpha=1, Translation="0,0,0", Rotation="0,0,0", Scale="1,1,1") {scaleMode=BMM_Replace, boneToModify {boneName=spine_01}}
     Alpha=1
     Translation="0,0,0"
     Rotation="0,0,0"
     Scale="1,1,1"
    ) {scaleMode=BMM_Replace, boneToModify {boneName=spine_02}}
    Alpha=1
    Translation="0,0,0"
    Rotation="0,0,0"
    Scale="1,1,1"
   ) {scaleMode=BMM_Replace, boneToModify {boneName=spine_03}}
   n4 = Anim Dynamics - Bone: dyn_sleeve_r(
    ComponentPose="Transform (Modify) Bone - Bone: spine_05"(
     ComponentPose="Transform (Modify) Bone - Bone: spine_04"(ComponentPose=@n3, Alpha=1, Translation="0,0,0", Rotation="0,0,0", Scale="1,1,1") {scaleMode=BMM_Replace, boneToModify {boneName=spine_04}}
     Alpha=1
     Translation="0,0,0"
     Rotation="0,0,0"
     Scale="1,1,1"
    ) {scaleMode=BMM_Replace, boneToModify {boneName=spine_05}}
    Alpha=1
    ExternalForce="0,0,0"
   ) {
    angularDampingOverride=0.99
    angularSpringConstant=150
    bLinearSpring=true
    bOverrideAngularDamping=true
    bOverrideLinearDamping=true
    gravityScale=0.1
    linearDampingOverride=0.99
    linearSpringConstant=10
    simulationSpace=RootRelative
    alphaBoolBlend {blendOption=Linear}
    boundBone {boneName=dyn_sleeve_r}
    componentAppliedLinearAccClamp=<500 500 500>
    componentLinearAccScale=<0.7 0.7 0.7>
    componentLinearVelScale=<0.7 0.7 0.7>
    gravityOverride {z=-100}
    physicsBodyDefinitions [{
      boundBone {boneName=dyn_sleeve_r}
      boxExtents=<3 3 3>
      constraintSetup {angularLimitsMax=<0 2 2>, angularLimitsMin=<0 -2 -2>, angularTarget=<0 0 0>, linearAxesMax=<5 1 1>, linearAxesMin=<0 -1 -1>}
      localJointOffset=<-0.05 0 0>
     }]
    relativeSpaceBone {boneName=head}
    simSpaceSettings {maxAngularAcceleration=300, maxAngularVelocity=300}
   }
   n5 = Anim Dynamics - Bone: dyn_sleeve_bot_r(ComponentPose=@n4, Alpha=1, ExternalForce="0,0,0") {
    angularDampingOverride=0.99
    angularSpringConstant=150
    bLinearSpring=true
    bOverrideAngularDamping=true
    bOverrideLinearDamping=true
    gravityScale=0.1
    linearDampingOverride=0.99
    linearSpringConstant=10
    simulationSpace=RootRelative
    alphaBoolBlend {blendOption=Linear}
    boundBone {boneName=dyn_sleeve_bot_r}
    componentAppliedLinearAccClamp=<500 500 500>
    componentLinearAccScale=<0.7 0.7 0.7>
    componentLinearVelScale=<0.7 0.7 0.7>
    gravityOverride {z=-100}
    physicsBodyDefinitions [{
      boundBone {boneName=dyn_sleeve_bot_r}
      boxExtents=<1 1 1>
      constraintSetup {angularLimitsMax=<5 5 5>, angularLimitsMin=<0 -2 -5>, angularTarget=<0 0 0>, linearAxesMax=<2 4 4>, linearAxesMin=<0 0 -1>}
      localJointOffset=<-0.05 0 0>
     }]
    relativeSpaceBone {boneName=head}
    simSpaceSettings {maxAngularAcceleration=300, maxAngularVelocity=300}
   }
   n6 = Anim Dynamics - Bone: dyn_sleeve_l(ComponentPose=@n5, Alpha=1, ExternalForce="0,0,0") {
    angularDampingOverride=0.99
    angularSpringConstant=150
    bLinearSpring=true
    bOverrideAngularDamping=true
    bOverrideLinearDamping=true
    gravityScale=0.1
    linearDampingOverride=0.99
    linearSpringConstant=10
    simulationSpace=RootRelative
    alphaBoolBlend {blendOption=Linear}
    boundBone {boneName=dyn_sleeve_l}
    componentAppliedLinearAccClamp=<500 500 500>
    componentLinearAccScale=<0.7 0.7 0.7>
    componentLinearVelScale=<0.7 0.7 0.7>
    gravityOverride {z=-100}
    physicsBodyDefinitions [{
      boundBone {boneName=dyn_sleeve_l}
      boxExtents=<3 3 3>
      constraintSetup {angularLimitsMax=<0 2 2>, angularLimitsMin=<0 -2 -2>, angularTarget=<0 0 0>, linearAxesMax=<0 1 1>, linearAxesMin=<-5 -1 -1>}
      localJointOffset=<0.05 0 0>
     }]
    relativeSpaceBone {boneName=head}
    simSpaceSettings {maxAngularAcceleration=300, maxAngularVelocity=300}
   }
   n7 = Anim Dynamics - Bone: dyn_sleeve_bot_l(ComponentPose=@n6, Alpha=1, ExternalForce="0,0,0") {
    angularDampingOverride=0.99
    angularSpringConstant=150
    bLinearSpring=true
    bOverrideAngularDamping=true
    bOverrideLinearDamping=true
    gravityScale=0.1
    linearDampingOverride=0.99
    linearSpringConstant=10
    simulationSpace=RootRelative
    alphaBoolBlend {blendOption=Linear}
    boundBone {boneName=dyn_sleeve_bot_l}
    componentAppliedLinearAccClamp=<500 500 500>
    componentLinearAccScale=<0.7 0.7 0.7>
    componentLinearVelScale=<0.7 0.7 0.7>
    gravityOverride {z=-100}
    physicsBodyDefinitions [{
      boundBone {boneName=dyn_sleeve_bot_l}
      boxExtents=<1 1 1>
      constraintSetup {angularLimitsMax=<0 2 2>, angularLimitsMin=<-5 -5 -2>, angularTarget=<0 0 0>, linearAxesMax=<0 0 1>, linearAxesMin=<-5 0 -4>}
      localJointOffset=<0.05 0 0>
     }]
    relativeSpaceBone {boneName=head}
    simSpaceSettings {maxAngularAcceleration=300, maxAngularVelocity=300}
   }
   Output Pose(Result=Component To Local(ComponentPose=@n7)) {layerGroup=None}
  ]}
 Clamp Vector YZ {nodes [
   n1 = Inputs!() -> (InVec:Vector=~, Min:double=~, Max:double=1)
   n2 = BreakVector(InVec=@n1)
   "Clamp (Float)"(Value=@n2, Min=@n1.Min, Max=@n1.Max)
   Outputs!(ReturnValue:Vector=MakeVector(X=@n2, Y="Clamp (Float)"(Value=@n2.Y, Min=@n1.Min, Max=@n1.Max), Z="Clamp (Float)"(Value=@n2.Z, Min=@n1.Min, Max=@n1.Max)))
  ]}
 EventGraph {nodes [
   Event BlueprintUpdateAnimation()
   > n1 = Sequence()
   n2 = Sequence!(execute=@n1.then_1) -> (then_0, then_1, then_2:Exec, then_3:Exec, then_4:Exec)
   n3 = GetOwnerSkeletalMeshComponent()
   > Set Scale_spine_01(Scale_spine_01:Vector=Clamp Vector YZ!(
     InVec:Vector=GetSocketTransform!(self=@n3, InSocketName=spine_01, TransformSpace=RTS_ParentBoneSpace) -> (ReturnValue_Location:Vector, ReturnValue_Rotation:Rotator, ReturnValue_Scale:Vector).ReturnValue_Scale
     Min:double
     Max:double=1
    ) -> (ReturnValue:Vector=~) {is=title K2Node_MacroInstance}) -> (Output_Get:Vector)
   Set Scale_spine_02(
    execute=@n2.then_1
    Scale_spine_02:Vector=Clamp Vector YZ(InVec=GetSocketTransform(self=@n3, InSocketName=spine_02, TransformSpace=RTS_ParentBoneSpace).ReturnValue_Scale, Max=1) -> (ReturnValue=~)
   ) -> (Output_Get:Vector)
   Set Scale_spine_03(
    execute=@n2.then_2
    Scale_spine_03:Vector=Clamp Vector YZ(InVec=GetSocketTransform(self=@n3, InSocketName=spine_03, TransformSpace=RTS_ParentBoneSpace).ReturnValue_Scale, Max=1) -> (ReturnValue=~)
   ) -> (Output_Get:Vector)
   Set Scale_spine_04(
    execute=@n2.then_3
    Scale_spine_04:Vector=Clamp Vector YZ(InVec=GetSocketTransform(self=@n3, InSocketName=spine_04, TransformSpace=RTS_ParentBoneSpace).ReturnValue_Scale, Max=1) -> (ReturnValue=~)
   ) -> (Output_Get:Vector)
   Set Scale_spine_05(
    execute=@n2.then_4
    Scale_spine_05:Vector=Clamp Vector YZ(InVec=GetSocketTransform(self=@n3, InSocketName=spine_05, TransformSpace=RTS_ParentBoneSpace).ReturnValue_Scale, Max=1) -> (ReturnValue=~)
   ) -> (Output_Get:Vector)
   n4 = GetPartSkeletalMeshComponent(self=Self-Reference())
   Set Sleeve_L_Alpha(
    execute=@n1.then_0
    Sleeve_L_Alpha:double=Remap Float(
     Input=GetSocketTransform!(self=@n4, InSocketName=upperarm_l, TransformSpace=RTS_ParentBoneSpace) -> (ReturnValue_Location, ReturnValue_Rotation_Roll:float, ReturnValue_Rotation_Pitch:float, ReturnValue_Rotation_Yaw:float, ReturnValue_Scale).ReturnValue_Rotation_Roll
     OldRange=(X=10,Y=50)
     NewRange=(X=1,Y=0)
     Clamp=true
    ) -> (Output=~)
   ) -> (Output_Get:double)
   > Set Sleeve_R_Alpha(Sleeve_R_Alpha:double=Remap Float(
     Input=GetSocketTransform(self=@n4, InSocketName=upperarm_r, TransformSpace=RTS_ParentBoneSpace).ReturnValue_Rotation_Roll
     OldRange=(X=10,Y=50)
     NewRange=(X=1,Y=0)
     Clamp=true
    ) -> (Output=~)) -> (Output_Get:double)
  ]}
]
variables [
 Scale_spine_01 {type=Vector}
 Scale_spine_02 {type=Vector}
 Scale_spine_03 {type=Vector}
 Scale_spine_04 {type=Vector}
 Scale_spine_05 {type=Vector}
 Sleeve_L_Alpha {type=number, default=1}
 Sleeve_R_Alpha {type=number, default=1}
]
