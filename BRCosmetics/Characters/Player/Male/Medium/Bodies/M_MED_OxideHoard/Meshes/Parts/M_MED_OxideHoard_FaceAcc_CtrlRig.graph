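# Face accessory Control Rig for the OxideHoard medium male body that shifts the Glasses bone to follow facial expressions.
# Nasolabial furrow positions relative to the head are remapped, and the larger side weights a glasses rotation.
# A second transform then adds a slight rotation and shift, weighted by the larger remapped displacement of the inner forehead bones from their initial pose.
# The correctives are driven purely by bone motion, with no controls or variables.
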ControlRigBlueprint /BRCosmetics/Characters/Player/Male/Medium/Bodies/M_MED_OxideHoard/Meshes/Parts/M_MED_OxideHoard_FaceAcc_CtrlRig
graphs [
 RigVMFunctionLibrary {links [], nodes []}
 RigVMModel {nodes [
   RigUnit_BeginExecution = &RigUnit_BeginExecution()
   RigUnit_GetRelativeTransformForItem = &RigUnit_GetRelativeTransformForItem(Child=(Type=Bone,Name="FACIAL_L_NasolabialFurrow"), Parent=(Type=Bone,Name="head"), bChildInitial=False, bParentInitial=False)
   Remap = &RigVMFunction_MathDoubleRemap(
    SourceMaximum=4.8
    SourceMinimum=1
    TargetMaximum=1
    TargetMinimum=0
    bClamp=False
    Value=@RigUnit_GetRelativeTransformForItem.RelativeTransform.Translation.X
   )
   n1 = &RigUnit_GetRelativeTransformForItem(Child=(Type=Bone,Name="FACIAL_R_NasolabialFurrow"), Parent=(Type=Bone,Name="head"), bChildInitial=False, bParentInitial=False)
   Maximum = &RigVMFunction_MathDoubleMax(
    A=@Remap.Result
    B=&RigVMFunction_MathDoubleRemap(SourceMaximum=4.8, SourceMinimum=1, TargetMaximum=1, TargetMinimum=0, bClamp=False, Value=@n1.RelativeTransform.Translation.X).Result
   )
   &RigUnit_ModifyTransforms(
    ItemToModify=((Item=(Type=Bone,Name="Glasses"),Transform=(Rotation=(X=0,Y=0,Z=-0.0305,W=0.9995),Translation=(X=0,Y=0,Z=0),Scale3D=(X=1,Y=1,Z=1))))
    WeightMinimum=0
    Weight=@Maximum.Result
    ExecutePin=@RigUnit_BeginExecution.ExecutePin
   )
   n2 = &RigUnit_GetRelativeTransformForItem(Child=(Type=Bone,Name="FACIAL_L_ForeheadIn"), Parent=(Type=Bone,Name="FACIAL_L_ForeheadIn"), bChildInitial=False, bParentInitial=true)
   n3 = &RigUnit_GetRelativeTransformForItem(Child=(Type=Bone,Name="FACIAL_R_ForeheadIn"), Parent=(Type=Bone,Name="FACIAL_R_ForeheadIn"), bChildInitial=False, bParentInitial=true)
   > &RigUnit_ModifyTransforms(
    ItemToModify=((Item=(Type=Bone,Name="Glasses"),Transform=(Rotation=(X=0,Y=0,Z=0.0044,W=1),Translation=(X=0,Y=0.1,Z=0),Scale3D=(X=1,Y=1,Z=1))))
    WeightMinimum=0
    Weight=&RigVMFunction_MathDoubleMax(
     A=&RigVMFunction_MathDoubleRemap(SourceMaximum=1.19, SourceMinimum=0.75, TargetMaximum=1, TargetMinimum=0, Value=@n2.RelativeTransform.Translation.Y).Result
     B=&RigVMFunction_MathDoubleRemap(SourceMaximum=1.19, SourceMinimum=0.75, TargetMaximum=1, TargetMinimum=0, Value=@n3.RelativeTransform.Translation.Y).Result
    ).Result
   )
  ]}
]
hierarchy [
 root {loc=<0 0 0>, rot=<0 0 0 1>}
 pelvis {parent=root, loc=<0 -0.7071 93.1191>, rot=<0.0127 0.707 -0.0127 -0.707>}
 spine_01 {parent=pelvis, loc=<6.7494 -0.6388 0>, rot=<0 0 0.034 -0.9994>}
 spine_02 {parent=spine_01, loc=<7.2518 -0.035 0>, rot=<0 0 0.0238 -0.9997>}
 spine_03 {parent=spine_02, loc=<6.0104 0 0>, rot=<0 0 0.0149 -0.9999>}
 spine_04 {parent=spine_03, loc=<8.1024 0 0>, rot=<0 0 -0.0132 -0.9999>}
 spine_05 {parent=spine_04, loc=<10.0146 0 0>, rot=<0 0 -0.0015 -1>}
 neck_01 {parent=spine_05, loc=<11.3138 0.1529 0>, rot=<0 0 -0.2003 -0.9797>}
 neck_02 {parent=neck_01, loc=<4.5627 0 0>, rot=<0 0 0.0504 -0.9987>}
 head {parent=neck_02, loc=<5.3914 0 0>, rot=<0 0 0.115 -0.9934>}
 FACIAL_C_FacialRoot {parent=head, loc=<2.1524 -0.746 -0.0002>, rot=<0.4879 -0.5118 -0.4879 -0.5118>}
 FACIAL_C_Forehead {parent=FACIAL_C_FacialRoot, loc=<-0.0033 -3.9647 10.5968>, rot=<-0.0437 -0.0004 -0.0001 -0.999>}
 FACIAL_L_ForeheadIn {parent=FACIAL_C_FacialRoot, loc=<1.5279 -3.9117 10.5605>, rot=<-0.061 -0.07 0.0042 -0.9957>}
 FACIAL_R_ForeheadIn {parent=FACIAL_C_FacialRoot, loc=<-1.5357 -3.8977 10.568>, rot=<-0.061 0.0692 -0.0043 -0.9957>}
 FACIAL_L_ForeheadMid {parent=FACIAL_C_FacialRoot, loc=<2.9527 -4.1762 10.1299>, rot=<-0.0713 -0.176 -0.0207 -0.9816>}
 FACIAL_R_ForeheadMid {parent=FACIAL_C_FacialRoot, loc=<-2.9683 -4.1604 10.1291>, rot=<-0.0713 0.1753 0.0206 -0.9817>}
 FACIAL_L_ForeheadOut {parent=FACIAL_C_FacialRoot, loc=<4.2366 -4.0853 9.3894>, rot=<-0.1216 -0.2559 -0.0793 -0.9557>}
 FACIAL_R_ForeheadOut {parent=FACIAL_C_FacialRoot, loc=<-4.2571 -4.0767 9.3863>, rot=<-0.1216 0.2552 0.0792 -0.9559>}
 FACIAL_L_EyesackUpper {parent=FACIAL_C_FacialRoot, loc=<2.9348 -3.1423 10.5688>, rot=<0.0485 -0.0846 -0.0647 -0.9931>}
 FACIAL_R_EyesackUpper {parent=FACIAL_C_FacialRoot, loc=<-2.9646 -3.1295 10.5773>, rot=<0.0485 0.0839 0.0644 -0.9932>}
 FACIAL_L_EyelidUpperFurrow {parent=FACIAL_C_FacialRoot, loc=<2.7424 -3.025 10.3338>, rot=<-0.0303 -0.1512 -0.0798 -0.9848>}
 FACIAL_R_EyelidUpperFurrow {parent=FACIAL_C_FacialRoot, loc=<-2.7668 -3.0042 10.3564>, rot=<-0.0303 0.1504 0.0796 -0.9849>}
 FACIAL_L_EyelidUpperA {parent=FACIAL_C_FacialRoot, loc=<2.7199 -2.633 8.9166>, rot=<-0.1481 -0.0004 -0.0001 -0.989>}
 FACIAL_R_EyelidUpperA {parent=FACIAL_C_FacialRoot, loc=<-2.7534 -2.6279 8.9194>, rot=<-0.1481 -0.0004 -0.0001 -0.989>}
 FACIAL_L_EyelidLowerA {parent=FACIAL_C_FacialRoot, loc=<2.7199 -2.633 8.9166>, rot=<0.2837 -0.0003 -0.0002 -0.9589>}
 FACIAL_R_EyelidLowerA {parent=FACIAL_C_FacialRoot, loc=<-2.7534 -2.6279 8.9194>, rot=<0.2837 -0.0003 -0.0002 -0.9589>}
 FACIAL_L_EyeCornerInner {parent=FACIAL_C_FacialRoot, loc=<1.8226 -2.1956 9.659>, rot=<-0.0003 -0.0004 -0.0001 -1>}
 FACIAL_R_EyeCornerInner {parent=FACIAL_C_FacialRoot, loc=<-1.8534 -2.1967 9.6503>, rot=<-0.0003 -0.0004 -0.0001 -1>}
 FACIAL_L_EyeCornerOuter {parent=FACIAL_C_FacialRoot, loc=<3.9912 -2.2621 9.2873>, rot=<-0.0002 -0.3829 -0.0002 -0.9238>}
 FACIAL_R_EyeCornerOuter {parent=FACIAL_C_FacialRoot, loc=<-4.0552 -2.2507 9.2627>, rot=<-0.0003 0.3823 0 -0.9241>}
 FACIAL_L_EyesackLower {parent=FACIAL_C_FacialRoot, loc=<2.9602 -1.1414 9.176>, rot=<-0.0226 -0.1401 -0.0224 -0.9896>}
 FACIAL_R_EyesackLower {parent=FACIAL_C_FacialRoot, loc=<-2.9826 -1.1304 9.1583>, rot=<-0.0226 0.1393 0.0222 -0.9897>}
 FACIAL_L_CheekInner {parent=FACIAL_C_FacialRoot, loc=<1.9986 -0.3986 9.2982>, rot=<-0.1236 -0.0883 -0.1972 -0.9685>}
 FACIAL_R_CheekInner {parent=FACIAL_C_FacialRoot, loc=<-2.0155 -0.3862 9.2956>, rot=<-0.1235 0.0876 0.1971 -0.9686>}
 FACIAL_L_CheekOuter {parent=FACIAL_C_FacialRoot, loc=<3.688 0.0015 7.7326>, rot=<0.055 -0.3773 0.1001 -0.919>}
 FACIAL_R_CheekOuter {parent=FACIAL_C_FacialRoot, loc=<-3.6971 0.0056 7.7193>, rot=<0.0548 0.3766 -0.1004 -0.9193>}
 FACIAL_L_NasolabialBulge {parent=FACIAL_C_FacialRoot, loc=<3.3581 3.2908 7.9809>, rot=<0.0577 -0.3777 0.0236 -0.9238>}
 FACIAL_R_NasolabialBulge {parent=FACIAL_C_FacialRoot, loc=<-3.3807 3.2919 7.9443>, rot=<0.0576 0.377 -0.0238 -0.9241>}
 FACIAL_L_NasolabialFurrow {parent=FACIAL_C_FacialRoot, loc=<2.5211 1.9978 10.0416>, rot=<-0.0337 -0.1261 -0.3638 -0.9223>}
 FACIAL_R_NasolabialFurrow {parent=FACIAL_C_FacialRoot, loc=<-2.5857 2.0344 10.0408>, rot=<-0.0334 0.1254 0.3636 -0.9225>}
 FACIAL_L_CheekLower {parent=FACIAL_C_FacialRoot, loc=<4.2018 2.6958 7.0547>, rot=<0.0344 -0.4272 0.0575 -0.9017>}
 FACIAL_R_CheekLower {parent=FACIAL_C_FacialRoot, loc=<-4.2148 2.6973 7.0113>, rot=<0.0343 0.4266 -0.0577 -0.902>}
 FACIAL_C_Nose {parent=FACIAL_C_FacialRoot, loc=<-0.0044 -1.0344 10.9628>, rot=<-0.1306 -0.0004 -0.0001 -0.9914>}
 FACIAL_C_NoseLower {parent=FACIAL_C_Nose, loc=<0.0095 2.6533 -0.3624>, rot=<0.1305 0 0 -0.9914>}
 FACIAL_L_Nostril {parent=FACIAL_C_Nose, loc=<0.9176 2.0536 -0.4346>, rot=<0.0173 0.1305 0.0023 -0.9913>}
 FACIAL_R_Nostril {parent=FACIAL_C_Nose, loc=<-0.8873 2.0804 -0.4295>, rot=<0.0173 -0.1305 -0.0023 -0.9913>}
 FACIAL_C_NoseTip {parent=FACIAL_C_Nose, loc=<0.0053 1.7903 0.4905>, rot=<0.0872 0 0 -0.9962>}
 FACIAL_C_MouthUpper {parent=FACIAL_C_FacialRoot, loc=<-0.0207 3.381 8.8295>, rot=<-0.0003 -0.0004 -0.0001 -1>}
 FACIAL_C_LipUpper {parent=FACIAL_C_MouthUpper, loc=<-0.0111 -0.613 2.3619>, rot=<-0.0872 0 0 -0.9962>}
 FACIAL_L_LipUpper {parent=FACIAL_C_MouthUpper, loc=<1.1617 -0.4427 2.0475>, rot=<-0.0778 -0.1601 -0.0379 -0.9833>}
 FACIAL_R_LipUpper {parent=FACIAL_C_MouthUpper, loc=<-1.1828 -0.4319 2.0236>, rot=<-0.0778 0.1601 0.0379 -0.9833>}
 FACIAL_L_LipUpperOuter {parent=FACIAL_C_MouthUpper, loc=<1.8469 -0.0654 1.3614>, rot=<-0.0984 -0.2293 -0.0983 -0.9634>}
 FACIAL_R_LipUpperOuter {parent=FACIAL_C_MouthUpper, loc=<-1.8907 -0.0553 1.3068>, rot=<-0.0984 0.2293 0.0983 -0.9634>}
 FACIAL_L_LipCorner {parent=FACIAL_C_MouthUpper, loc=<2.0696 0.288 0.7829>, rot=<0 -0.2078 0 -0.9782>}
 FACIAL_R_LipCorner {parent=FACIAL_C_MouthUpper, loc=<-2.1132 0.2891 0.745>, rot=<0 0.2078 0 -0.9782>}
 FACIAL_L_Masseter {parent=FACIAL_C_FacialRoot, loc=<6.276 0.9181 5.2186>, rot=<0.0691 0.1125 0.0078 -0.9912>}
 FACIAL_R_Masseter {parent=FACIAL_C_FacialRoot, loc=<-6.2729 0.9481 5.2207>, rot=<0.0691 -0.1132 -0.0081 -0.9911>}
 FACIAL_C_UnderChin {parent=FACIAL_C_FacialRoot, loc=<-0.0073 8.9323 6.5357>, rot=<0.4996 -0.0003 -0.0003 -0.8663>}
 FACIAL_L_UnderChin {parent=FACIAL_C_FacialRoot, loc=<3.5566 7.3287 4.6566>, rot=<0.4024 -0.3552 0.082 -0.8398>}
 FACIAL_R_UnderChin {parent=FACIAL_C_FacialRoot, loc=<-3.5952 7.3492 4.6626>, rot=<0.4022 0.3546 -0.0825 -0.84>}
 FACIAL_C_LowerLipRotation {parent=FACIAL_C_FacialRoot, loc=<-0.0147 0.5941 3.0151>, rot=<0.3418 -0.0003 -0.0002 -0.9398>}
 FACIAL_C_MouthLower {parent=FACIAL_C_LowerLipRotation, loc=<-0.0096 -1.5992 6.2463>, rot=<-0.342 0 0 -0.9397>}
 FACIAL_C_LipLower {parent=FACIAL_C_MouthLower, loc=<-0.0275 0.8957 1.7905>, rot=<0.0872 0 0 -0.9962>}
 FACIAL_L_LipLower {parent=FACIAL_C_MouthLower, loc=<0.9306 0.8475 1.727>, rot=<0.0927 -0.1294 0.0272 -0.9869>}
 FACIAL_R_LipLower {parent=FACIAL_C_MouthLower, loc=<-0.9728 0.8505 1.6998>, rot=<0.0927 0.1294 -0.0272 -0.9869>}
 FACIAL_L_LipLowerOuter {parent=FACIAL_C_MouthLower, loc=<1.7155 0.6014 1.3625>, rot=<0.0913 -0.2321 0.0318 -0.9679>}
 FACIAL_R_LipLowerOuter {parent=FACIAL_C_MouthLower, loc=<-1.7619 0.627 1.3188>, rot=<0.0913 0.2321 -0.0318 -0.9679>}
 FACIAL_C_Jaw {parent=FACIAL_C_FacialRoot, loc=<-0.0147 0.594 3.0151>, rot=<0.3418 -0.0003 -0.0002 -0.9398>}
 FACIAL_C_Jawline {parent=FACIAL_C_Jaw, loc=<-0.0638 2.3325 8.9418>, rot=<0.2586 0 0 -0.966>}
 FACIAL_L_ChinSide {parent=FACIAL_C_Jaw, loc=<2.7892 0.254 8.3389>, rot=<-0.2208 -0.3738 0.0499 -0.8995>}
 FACIAL_R_ChinSide {parent=FACIAL_C_Jaw, loc=<-2.8225 0.2942 8.2674>, rot=<-0.2208 0.3738 -0.0499 -0.8995>}
 FACIAL_C_Chin {parent=FACIAL_C_Jaw, loc=<-0.0525 0.702 8.8011>, rot=<-0.2591 0 0 -0.9659>}
 FACIAL_L_Ear {parent=FACIAL_C_FacialRoot, loc=<6.8414 -1.0228 1.8591>, rot=<-0.0266 -0.001 -0.0263 -0.9993>}
 FACIAL_R_Ear {parent=FACIAL_C_FacialRoot, loc=<-6.8391 -0.9524 1.8997>, rot=<-0.0266 0.0003 0.0261 -0.9993>}
 FACIAL_L_Jawline {parent=FACIAL_C_FacialRoot, loc=<4.4645 5.0738 5.9163>, rot=<0.209 0.1574 -0.0548 -0.9636>}
 FACIAL_R_Jawline {parent=FACIAL_C_FacialRoot, loc=<-4.4753 5.0768 5.8806>, rot=<0.2091 -0.1581 0.0544 -0.9635>}
 FACIAL_C_NoseBridge {parent=FACIAL_C_FacialRoot, loc=<-0.0021 -2.6967 11.4297>, rot=<0.0435 -0.0004 -0.0001 -0.9991>}
 FACIAL_L_EyelidUpperB {parent=FACIAL_C_FacialRoot, loc=<2.7199 -2.633 8.9166>, rot=<-0.2252 -0.0004 0 -0.9743>}
 FACIAL_L_EyelidLowerB {parent=FACIAL_C_FacialRoot, loc=<2.7199 -2.633 8.9166>, rot=<0.317 -0.0003 -0.0002 -0.9484>}
 FACIAL_R_EyelidLowerB {parent=FACIAL_C_FacialRoot, loc=<-2.7534 -2.6279 8.9194>, rot=<0.3253 -0.0003 -0.0002 -0.9456>}
 FACIAL_R_EyelidUpperB {parent=FACIAL_C_FacialRoot, loc=<-2.7534 -2.6279 8.9194>, rot=<-0.2253 -0.0004 0 -0.9743>}
 hair_L {parent=head, loc=<14.404 8.3319 -1.2269>, rot=<-0.3109 -0.4659 0.8263 0.0585>}
 hair_R {parent=head, loc=<14.404 8.3319 1.2228>, rot=<-0.1225 -0.9444 0.0872 0.2924>}
 Glasses {parent=head, loc=<5.5699 1.4309 0>, rot=<0 0 0 1>}
 clavicle_l {parent=spine_05, loc=<4.3575 -0.1466 -2.3223>, rot=<0.0465 -0.7289 -0.009 -0.6829>}
 clavicle_r {parent=spine_05, loc=<4.3575 -0.1466 2.3223>, rot=<0.7289 0.0465 -0.6829 0.009>}
]
