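# Physics asset for the face accessory mesh of the M_MED_OxideHoard male character body, setting up physical movement for its left and right hair bones.
# Each of hair_L, hair_R and head has a single sphere body, with the larger one on the head.
# Both hair bodies are constrained to the head with identical, tight swing limits and almost no twist, restricting them to slight sway.
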PhysicsAsset /BRCosmetics/Characters/Player/Male/Medium/Bodies/M_MED_OxideHoard/Meshes/Parts/M_MED_OxideHoard_FaceAcc_Physics
bodies {
 hair_L [{shapeType=Sphere, radius=1.4235, shapeName=hair_L_sphere, center=<5.3352 1.6372 0.3972>}]
 hair_R [{shapeType=Sphere, radius=1.3621, shapeName=hair_R_sphere, center=<5.0563 -0.0615 -0.1432>}]
 head [{shapeType=Sphere, radius=7.19, shapeName=head_sphere, center=<8.4474 -0.468 0.0731>}]
}
constraints [
 {bone1Name=hair_L, bone2Name=head, swing1LimitDegrees=7, swing2LimitDegrees=15, twistLimitDegrees=1}
 {bone1Name=hair_R, bone2Name=head, swing1LimitDegrees=7, swing2LimitDegrees=15, twistLimitDegrees=1}
]
